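# UI material library function that blends a top and bottom color with Photoshop/Figma-style comparative modes: Difference, Exclusion, Subtract and Divide.
# Both inputs go through a square root as a cheap linear-to-sRGB conversion before the blend math runs.
# Each result is saturated, then raised to a shared gamma reroute, set to 2 instead of 2.2 for cheapness, and exposed as its own output.
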MaterialFunction /FortUILibrary/MaterialFunctions/Color/MF_UI_BlendModes_Comparative
graph {
 properties {
  bExposeToLibrary=true
  description=Allows you to blend colors similarly as Photoshop and Figma
  previewMaterialDomain=MD_UI
  libraryCategoriesText [UI_ShaderLibrary]
 }
 nodes [
  n1 = Abs(None=NamedRerouteUsage())
  FunctionOutput(None=Power(Base=Saturate(None=@n1), Exp=NamedRerouteUsage()) {constExponent=2.2}) {outputName=Difference, sortPriority=100}
  n2 = Subtract(A=NamedRerouteUsage()) {constB=0.5}
  n3 = Subtract(A=NamedRerouteUsage()) {constB=0.5}
  n4 = Abs(None=ConstantBiasScale(None=Multiply(A=@n2, B=@n3)) {bias=-0.25, scale=2})
  FunctionOutput(None=Power(Base=Saturate(None=@n4), Exp=NamedRerouteUsage()) {constExponent=2.2}) {outputName=Exclusion, sortPriority=105}
  n5 = NamedRerouteDeclaration(None=Subtract(A=NamedRerouteUsage(), B=NamedRerouteUsage())) {name=Subtracted Channels, nodeColor=<rgba 0 0 0 1>}
  FunctionOutput(None=Power(Base=Saturate(None=@n5), Exp=NamedRerouteUsage()) {constExponent=2.2}) {outputName=Subtract, sortPriority=110}
  n6 = Divide(A=NamedRerouteUsage(), B=Max(A=NamedRerouteUsage()) {constB=0})
  FunctionOutput(None=Power(Base=Saturate(None=@n6), Exp=NamedRerouteUsage()) {constExponent=2.2}) {outputName=Divide, sortPriority=115}
  n7 = FunctionInput() {bUsePreviewValueAsDefault=true, description=Top layer which will be used to blend, inputName=Top Layer, previewValue=<1 0 0.1276 1>}
  NamedRerouteDeclaration(None=SquareRoot(None=@n7) {desc=Cheap Linear to sRGB}) {name=Top, nodeColor=<rgba 0 0 0 1>}
  n8 = FunctionInput() {
   bUsePreviewValueAsDefault=true
   description=Bottom layer which is the base that will be influenced
   inputName=Bottom Layer
   sortPriority=10
   previewValue=<0.011 0.2285 1 1>
  }
  NamedRerouteDeclaration(None=SquareRoot(None=@n8) {desc=Cheap Linear to sRGB}) {name=Bottom, nodeColor=<rgba 0 0 0 1>}
  NamedRerouteDeclaration(None=Constant() {desc="Should be 2.2, but 2 to be cheaper", r=2}) {name=Gamma Correction, nodeColor=<rgba 0 0 0 1>}
 ]
}
