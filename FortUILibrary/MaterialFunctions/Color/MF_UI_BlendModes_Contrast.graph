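# UI material function library node that blends Top and Bottom color layers with Photoshop/Figma-style contrast modes: Overlay, Soft, Hard, Linear and Pin Light, and Hard Mix.
# Square roots cheaply convert both inputs from linear to sRGB, and named reroutes share intermediate terms.
# Most modes lerp between dark and light formulas; Hard Mix floors a sum.
# Each result is saturated and raised to power 2; one divide-based branch has no output.
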MaterialFunction /FortUILibrary/MaterialFunctions/Color/MF_UI_BlendModes_Contrast
graph {
 properties {
  bExposeToLibrary=true
  description=Allows you to blend colors similarly as Photoshop and Figma
  previewMaterialDomain=MD_UI
  libraryCategoriesText [UI_ShaderLibrary]
 }
 nodes [
  n1 = NamedRerouteDeclaration(None=Multiply(A=NamedRerouteUsage()) {constB=2}) {name=2xBottom, nodeColor=<rgba 0 0 0 1>}
  n2 = NamedRerouteUsage()
  n3 = Constant3Vector() {constant=<rgba 1 1 1 0>}
  n4 = NamedRerouteUsage()
  n5 = NamedRerouteUsage()
  n6 = Multiply(
   A=NamedRerouteDeclaration(None=OneMinus(None=@n4)) {name=1-Bottom, nodeColor=<rgba 0 0 0 1>}
   B=NamedRerouteDeclaration(None=OneMinus(None=@n5)) {name=1-Top, nodeColor=<rgba 0 0 0 1>}
  )
  n7 = LinearInterpolate(
   A=NamedRerouteDeclaration(None=Multiply(A=@n1, B=@n2)) {name=Overlay Lower Values, nodeColor=<rgba 0 0 0 1>}
   B=NamedRerouteDeclaration(None=LinearInterpolate(A=@n3, Alpha=@n6) {constA=1, constB=-1}) {name=Overlay Higher Values, nodeColor=<rgba 0 0 0 1>}
   Alpha=NamedRerouteUsage()
  )
  FunctionOutput(None=Power(Base=Saturate(None=@n7), Exp=NamedRerouteUsage()) {constExponent=2.2}) {outputName=Overlay, sortPriority=60}
  n8 = NamedRerouteUsage()
  n9 = NamedRerouteUsage()
  n10 = LinearInterpolate(A=Multiply(A=@n8, B=Add(A=@n9) {constB=0.5}), B=LinearInterpolate(A=Subtract(A=@n9) {constB=0.5}, Alpha=@n8), Alpha=NamedRerouteUsage())
  FunctionOutput(None=Power(Base=Saturate(None=@n10), Exp=NamedRerouteUsage()) {constExponent=2.2}) {outputName=Soft Light, sortPriority=65}
  n11 = LinearInterpolate(A=NamedRerouteUsage(), B=NamedRerouteUsage(), Alpha=NamedRerouteUsage())
  FunctionOutput(None=Power(Base=Saturate(None=@n11), Exp=NamedRerouteUsage()) {constExponent=2.2}) {outputName=Hard Light, sortPriority=70}
  n12 = NamedRerouteDeclaration(None=Multiply(A=NamedRerouteUsage()) {constB=2}) {name=2xTop, nodeColor=<rgba 0 0 0 1>}
  n13 = NamedRerouteUsage()
  n14 = ConstantBiasScale(None=NamedRerouteUsage()) {bias=-0.5, scale=2}
  n15 = LinearInterpolate(
   A=Subtract(A=Add(A=@n12, B=@n13))
   B=Add(A=NamedRerouteDeclaration(None=@n14) {name=Shifted Top, nodeColor=<rgba 0 0 0 1>}, B=NamedRerouteUsage())
   Alpha=NamedRerouteUsage()
  )
  FunctionOutput(None=Power(Base=Saturate(None=@n15), Exp=NamedRerouteUsage()) {constExponent=2.2}) {outputName=Linear Light, sortPriority=80}
  n16 = NamedRerouteUsage()
  n17 = LinearInterpolate(A=Min(A=NamedRerouteUsage(), B=@n16), B=Max(A=@n16, B=NamedRerouteUsage()) {constB=0}, Alpha=NamedRerouteUsage())
  FunctionOutput(None=Power(Base=Saturate(None=@n17), Exp=NamedRerouteUsage()) {constExponent=2.2}) {outputName=Pin Light, sortPriority=85}
  n18 = Floor(None=Add(A=NamedRerouteUsage(), B=NamedRerouteUsage()))
  FunctionOutput(None=Power(Base=Saturate(None=@n18), Exp=NamedRerouteUsage()) {constExponent=2.2}) {outputName=Hard Mix, sortPriority=90}
  n19 = FunctionInput() {bUsePreviewValueAsDefault=true, description=Top layer which will be used to blend, inputName=Top Layer, previewValue=<1 0 0.1276 1>}
  n20 = SquareRoot(None=@n19) {desc=Cheap Linear to sRGB}
  NamedRerouteDeclaration(None=Round(None=NamedRerouteDeclaration(None=@n20) {name=Top, nodeColor=<rgba 0 0 0 1>})) {name=Rounded Top Channels, nodeColor=<rgba 0 0 0 1>}
  n21 = FunctionInput() {
   bUsePreviewValueAsDefault=true
   description=Bottom layer which is the base that will be influenced
   inputName=Bottom Layer
   sortPriority=10
   previewValue=<0.011 0.2285 1 1>
  }
  n22 = SquareRoot(None=@n21) {desc=Cheap Linear to sRGB}
  NamedRerouteDeclaration(None=Round(None=NamedRerouteDeclaration(None=@n22) {name=Bottom, nodeColor=<rgba 0 0 0 1>})) {name=Rounded Bottom Channels, nodeColor=<rgba 0 0 0 1>}
  NamedRerouteDeclaration(None=Constant() {desc="Should be 2.2, but 2 to be cheaper", r=2}) {name=Gamma Correction, nodeColor=<rgba 0 0 0 1>}
  n23 = OneMinus(None=Divide(A=NamedRerouteUsage(), B=NamedRerouteUsage()))
  n24 = NamedRerouteUsage()
  n25 = Divide(A=NamedRerouteUsage(), B=Max(A=Multiply(A=@n24) {constB=2}) {constB=0})
  n26 = NamedRerouteUsage()
  Power(Base=Saturate(None=LinearInterpolate(A=@n23, B=@n25, Alpha=@n26)), Exp=NamedRerouteUsage()) {constExponent=2.2}
 ]
}
