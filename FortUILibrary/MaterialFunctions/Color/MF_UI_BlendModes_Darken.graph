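# UI shader library material function that blends Top and Bottom Layer colors with Photoshop and Figma-style darkening modes.
# Both inputs are square-rooted as a cheap linear-to-sRGB conversion, blended, then converted back with a power of 2 rather than 2.2 to save cost.
# Separate outputs provide Darken, Multiply, Color Burn, Plus Darker and Darker Color.
# Darker Color picks a whole layer by comparing desaturated luminance, while Darken takes the per-channel minimum.
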MaterialFunction /FortUILibrary/MaterialFunctions/Color/MF_UI_BlendModes_Darken
graph {
 properties {
  bExposeToLibrary=true
  description=Allows you to blend colors similarly as Photoshop and Figma
  previewMaterialDomain=MD_UI
  libraryCategoriesText [UI_ShaderLibrary]
 }
 nodes [
  n1 = NamedRerouteUsage()
  n2 = NamedRerouteUsage()
  FunctionOutput(None=Power(Base=Min(A=@n1, B=@n2), Exp=NamedRerouteUsage()) {constExponent=2.2}) {outputName=Darken}
  n3 = NamedRerouteUsage()
  n4 = NamedRerouteUsage()
  n5 = LinearInterpolate(
   A=NamedRerouteUsage()
   B=NamedRerouteUsage()
   Alpha=Step(
    Y=Desaturation(None=@n3) {luminanceFactors=<rgba 0.299 0.587 0.114 1>}
    X=Desaturation(None=@n4) {luminanceFactors=<rgba 0.299 0.587 0.114 1>}
   )
  )
  FunctionOutput(None=Power(Base=Saturate(None=@n5), Exp=NamedRerouteUsage()) {constExponent=2.2}) {outputName=Darker Color, sortPriority=20}
  n6 = Multiply(A=NamedRerouteUsage(), B=NamedRerouteUsage())
  FunctionOutput(None=Power(Base=Saturate(None=@n6), Exp=NamedRerouteUsage()) {constExponent=2.2}) {outputName=Multiply, sortPriority=5}
  n7 = NamedRerouteUsage()
  n8 = NamedRerouteUsage()
  n9 = OneMinus(None=Divide(A=OneMinus(None=@n7), B=Max(A=@n8) {constB=0}))
  FunctionOutput(None=Power(Base=Saturate(None=@n9), Exp=NamedRerouteUsage()) {constExponent=2.2}) {outputName=Color Burn, sortPriority=10}
  n10 = Subtract(A=Add(A=NamedRerouteUsage(), B=NamedRerouteUsage()))
  FunctionOutput(None=Power(Base=Saturate(None=@n10), Exp=NamedRerouteUsage()) {constExponent=2.2}) {outputName=Plus Darker, sortPriority=15}
  n11 = FunctionInput() {bUsePreviewValueAsDefault=true, description=Top layer which will be used to blend, inputName=Top Layer, previewValue=<1 0 0.1276 1>}
  NamedRerouteDeclaration(None=SquareRoot(None=@n11) {desc=Cheap Linear to sRGB}) {name=Top, nodeColor=<rgba 0 0 0 1>}
  n12 = FunctionInput() {
   bUsePreviewValueAsDefault=true
   description=Bottom layer which is the base that will be influenced
   inputName=Bottom Layer
   sortPriority=10
   previewValue=<0.011 0.2285 1 1>
  }
  NamedRerouteDeclaration(None=SquareRoot(None=@n12) {desc=Cheap Linear to sRGB}) {name=Bottom, nodeColor=<rgba 0 0 0 1>}
  NamedRerouteDeclaration(None=Constant() {desc="Should be 2.2, but 2 to be cheaper", r=2}) {name=Gamma Correction, nodeColor=<rgba 0 0 0 1>}
 ]
}
