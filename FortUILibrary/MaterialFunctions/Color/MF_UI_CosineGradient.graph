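# UI material function that generates a procedural color gradient from a cosine palette.
# Computes saturate(a + b * cos(c * Gradient + d)), where inputs a, b, c and d are RGB offset, amplitude, frequency and phase.
# Gradient defaults to the horizontal component of the normalized UI UV.
# Preview values double as defaults, with staggered per-channel phases giving a rainbow-style ramp when nothing is connected.
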MaterialFunction /FortUILibrary/MaterialFunctions/Color/MF_UI_CosineGradient
graph {nodes [
  n1 = Constant() {r=0.5}
  n2 = Constant() {r=0.5}
  n3 = Constant() {r=0.5}
  n4 = Constant() {r=1}
  n5 = Constant() {r=1}
  n6 = Constant() {r=1}
  n7 = MaterialFunctionCall() {materialFunction=&GetUserInterfaceUV}
  n8 = Multiply(
   A=FunctionInput(Preview=MaterialFunctionCall(R=@n4, G=@n5, B=@n6) {desc=c, materialFunction=&AppendMany}.RGB) {bUsePreviewValueAsDefault=true, inputName=c, sortPriority=3}
   B=FunctionInput(Preview=MaterialFunctionCall(Float2=@n7.Normalized UV) {materialFunction=&BreakOutFloat2Components}.R) {bUsePreviewValueAsDefault=true, inputName=Gradient, inputType=FunctionInput_Vector2}
  )
  n9 = FunctionInput(Preview=MaterialFunctionCall(R=Constant(), G=Constant() {r=0.33}, B=Constant() {r=0.67}) {desc=d, materialFunction=&AppendMany}.RGB) {bUsePreviewValueAsDefault=true, inputName=d, sortPriority=4}
  n10 = Multiply(
   A=FunctionInput(Preview=MaterialFunctionCall(R=@n1, G=@n2, B=@n3) {desc=b, materialFunction=&AppendMany}.RGB) {bUsePreviewValueAsDefault=true, inputName=b, sortPriority=2}
   B=Cosine(None=Add(A=@n8, B=@n9))
  )
  n11 = FunctionInput(Preview=MaterialFunctionCall(R=Constant() {r=0.5}, G=Constant() {r=0.5}, B=Constant() {r=0.5}) {desc=a, materialFunction=&AppendMany}.RGB) {bUsePreviewValueAsDefault=true, inputName=a, sortPriority=1}
  FunctionOutput(None=Saturate(None=Add(A=@n10, B=@n11) {desc=a}))
 ]}
